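# Anim Blueprint for the IvyStub mic stand cosmetic that makes its small and large stone bones float, spin and sway.
# Each update turns game time into sine/cosine offset vectors and steadily growing rotators.
# Chained additive bone-space Modify Bone nodes share these values across the stones, though stone_04_large gets zero offset.
# Anim Dynamics springs relative to the handle bone then add damped, gravity-free lag to every modified stone except stone_04_large.
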AnimBlueprint /SparksCosmetics/Instrument/MicStand/IvyStub/Meshes/IvyStub_MicStand_AnimBP
target_skeleton=&/IvyStub_MicStand_Skeleton
graphs [
 AnimGraph {nodes [
   n1 = Get StoneFloatA() -> (StoneFloatA:Vector)
   n2 = Get Rotator1() -> (Rotator1:Rotator)
   n3 = "Transform (Modify) Bone - Bone: stone_01_small"(ComponentPose=Local To Component(LocalPose=Input Pose()), Alpha=1, Translation=@n1, Rotation=@n2, Scale="1,1,1") {
    rotationMode=BMM_Additive
    translationMode=BMM_Additive
    translationSpace=BCS_BoneSpace
    alphaBoolBlend {blendOption=Linear}
    boneToModify {boneName=stone_01_small}
   }
   n4 = Get StoneFloatB() -> (StoneFloatB:Vector)
   n5 = Get Rotator2() -> (Rotator2:Rotator)
   n6 = "Transform (Modify) Bone - Bone: stone_02_small"(ComponentPose=@n3, Alpha=1, Translation=@n4, Rotation=@n5, Scale="1,1,1") {
    rotationMode=BMM_Additive
    translationMode=BMM_Additive
    translationSpace=BCS_BoneSpace
    alphaBoolBlend {blendOption=Linear}
    boneToModify {boneName=stone_02_small}
   }
   n7 = Get StoneFloatC() -> (StoneFloatC:Vector)
   n8 = "Transform (Modify) Bone - Bone: stone_03_small"(ComponentPose=@n6, Alpha=1, Translation=@n7, Rotation=@n2, Scale="1,1,1") {
    rotationMode=BMM_Additive
    translationMode=BMM_Additive
    translationSpace=BCS_BoneSpace
    alphaBoolBlend {blendOption=Linear}
    boneToModify {boneName=stone_03_small}
   }
   n9 = Get Shard_Float_D() -> (Shard_Float_D:Vector)
   n10 = "Transform (Modify) Bone - Bone: stone_04_small"(ComponentPose=@n8, Alpha=1, Translation=@n9, Rotation=@n5, Scale="1,1,1") {
    rotationMode=BMM_Additive
    translationMode=BMM_Additive
    translationSpace=BCS_BoneSpace
    alphaBoolBlend {blendOption=Linear}
    boneToModify {boneName=stone_04_small}
   }
   n11 = "Transform (Modify) Bone - Bone: stone_05_small"(ComponentPose=@n10, Alpha=1, Translation=@n1, Rotation=@n2, Scale="1,1,1") {
    rotationMode=BMM_Additive
    translationMode=BMM_Additive
    translationSpace=BCS_BoneSpace
    alphaBoolBlend {blendOption=Linear}
    boneToModify {boneName=stone_05_small}
   }
   n12 = "Transform (Modify) Bone - Bone: stone_06_small"(ComponentPose=@n11, Alpha=1, Translation=@n4, Rotation=@n5, Scale="1,1,1") {
    rotationMode=BMM_Additive
    translationMode=BMM_Additive
    translationSpace=BCS_BoneSpace
    alphaBoolBlend {blendOption=Linear}
    boneToModify {boneName=stone_06_small}
   }
   n13 = "Transform (Modify) Bone - Bone: stone_07_small"(ComponentPose=@n12, Alpha=1, Translation=@n7, Rotation=@n2, Scale="1,1,1") {
    rotationMode=BMM_Additive
    translationMode=BMM_Additive
    translationSpace=BCS_BoneSpace
    alphaBoolBlend {blendOption=Linear}
    boneToModify {boneName=stone_07_small}
   }
   n14 = "Transform (Modify) Bone - Bone: stone_08_small"(ComponentPose=@n13, Alpha=1, Translation=@n9, Rotation=@n5, Scale="1,1,1") {
    rotationMode=BMM_Additive
    translationMode=BMM_Additive
    translationSpace=BCS_BoneSpace
    alphaBoolBlend {blendOption=Linear}
    boneToModify {boneName=stone_08_small}
   }
   n15 = "Transform (Modify) Bone - Bone: stone_09_small"(ComponentPose=@n14, Alpha=1, Translation=@n1, Rotation=@n2, Scale="1,1,1") {
    rotationMode=BMM_Additive
    translationMode=BMM_Additive
    translationSpace=BCS_BoneSpace
    alphaBoolBlend {blendOption=Linear}
    boneToModify {boneName=stone_09_small}
   }
   n16 = "Transform (Modify) Bone - Bone: stone_02_large"(ComponentPose=@n15, Alpha=1, Translation=Get StoneFloatB(), Rotation=Get Rotator3() -> (Rotator3:Rotator), Scale="1,1,1") {
    rotationMode=BMM_Additive
    translationMode=BMM_Additive
    translationSpace=BCS_BoneSpace
    alphaBoolBlend {blendOption=Linear}
    boneToModify {boneName=stone_02_large}
   }
   n17 = Get Rotator2()
   n18 = "Transform (Modify) Bone - Bone: stone_03_large"(ComponentPose=@n16, Alpha=1, Translation=Get StoneFloatA(), Rotation=@n17, Scale="1,1,1") {
    rotationMode=BMM_Additive
    translationMode=BMM_Additive
    translationSpace=BCS_BoneSpace
    alphaBoolBlend {blendOption=Linear}
    boneToModify {boneName=stone_03_large}
   }
   n19 = "Transform (Modify) Bone - Bone: stone_04_large"(ComponentPose=@n18, Alpha=1, Translation="0,0,0", Rotation="0,0,0", Scale="1,1,1") {
    rotationMode=BMM_Additive
    translationMode=BMM_Additive
    translationSpace=BCS_BoneSpace
    alphaBoolBlend {blendOption=Linear}
    boneToModify {boneName=stone_04_large}
   }
   n20 = "Transform (Modify) Bone - Bone: stone_05_large"(ComponentPose=@n19, Alpha=1, Translation=Get StoneFloatC(), Rotation=@n17, Scale="1,1,1") {
    rotationMode=BMM_Additive
    translationMode=BMM_Additive
    translationSpace=BCS_BoneSpace
    alphaBoolBlend {blendOption=Linear}
    boneToModify {boneName=stone_05_large}
   }
   n21 = Anim Dynamics - Bone: stone_02_large(ComponentPose=@n20, Alpha=0.8, ExternalForce="0,0,0") {
    angularDampingOverride=0.9
    bAngularSpring=true
    bGravityOverrideInSimSpace=true
    bLinearSpring=true
    bOverrideAngularDamping=true
    bOverrideLinearDamping=true
    gravityScale=0
    linearDampingOverride=0.9
    linearSpringConstant=10
    simulationSpace=BoneRelative
    alphaBoolBlend {blendOption=Linear}
    boundBone {boneName=stone_02_large}
    componentLinearAccScale=<0.2 0.2 0.2>
    componentLinearVelScale=<2 2 2>
    gravityOverride {z=-980}
    physicsBodyDefinitions [{
      sphereCollisionRadius=0
      boundBone {boneName=stone_02_large}
      boxExtents=<50 50 50>
      constraintSetup {angularLimitsMax=<0 0 0>, angularLimitsMin=<0 0 0>, angularTarget=<0 0 0>, linearAxesMax=<0.5 0.5 0.5>, linearAxesMin=<-0.5 -0.5 -0.5>}
      localJointOffset=<0 -3 0>
     }]
    relativeSpaceBone {boneName=handle}
    simSpaceSettings {maxAngularAcceleration=6000, maxAngularVelocity=6000}
   }
   n22 = Anim Dynamics - Bone: stone_03_large(ComponentPose=@n21, Alpha=0.8, ExternalForce="0,0,0") {
    angularDampingOverride=0.9
    bAngularSpring=true
    bGravityOverrideInSimSpace=true
    bLinearSpring=true
    bOverrideAngularDamping=true
    bOverrideLinearDamping=true
    gravityScale=0
    linearDampingOverride=0.9
    linearSpringConstant=10
    simulationSpace=BoneRelative
    alphaBoolBlend {blendOption=Linear}
    boundBone {boneName=stone_03_large}
    componentLinearAccScale=<0.2 0.2 0.2>
    componentLinearVelScale=<2 2 2>
    gravityOverride {z=-980}
    physicsBodyDefinitions [{
      sphereCollisionRadius=0
      boundBone {boneName=stone_03_large}
      boxExtents=<50 50 50>
      constraintSetup {angularLimitsMax=<0 0 0>, angularLimitsMin=<0 0 0>, angularTarget=<0 0 0>, linearAxesMax=<0.5 0.5 0.5>, linearAxesMin=<-0.5 -0.5 -0.5>}
      localJointOffset=<0 -3 0>
     }]
    relativeSpaceBone {boneName=handle}
    simSpaceSettings {maxAngularAcceleration=6000, maxAngularVelocity=6000}
   }
   n23 = Anim Dynamics - Bone: stone_05_large(ComponentPose=@n22, Alpha=0.8, ExternalForce="0,0,0") {
    angularDampingOverride=0.9
    bAngularSpring=true
    bGravityOverrideInSimSpace=true
    bLinearSpring=true
    bOverrideAngularDamping=true
    bOverrideLinearDamping=true
    gravityScale=0
    linearDampingOverride=0.9
    linearSpringConstant=10
    simulationSpace=BoneRelative
    alphaBoolBlend {blendOption=Linear}
    boundBone {boneName=stone_05_large}
    componentLinearAccScale=<0.2 0.2 0.2>
    componentLinearVelScale=<2 2 2>
    gravityOverride {z=-980}
    physicsBodyDefinitions [{
      sphereCollisionRadius=0
      boundBone {boneName=stone_05_large}
      boxExtents=<50 50 50>
      constraintSetup {angularLimitsMax=<0 0 0>, angularLimitsMin=<0 0 0>, angularTarget=<0 0 0>, linearAxesMax=<0.5 0.5 0.5>, linearAxesMin=<-0.5 -0.5 -0.5>}
      localJointOffset=<0 -3 0>
     }]
    relativeSpaceBone {boneName=handle}
    simSpaceSettings {maxAngularAcceleration=6000, maxAngularVelocity=6000}
   }
   n24 = Anim Dynamics - Bone: stone_01_small(ComponentPose=@n23, Alpha=0.8, ExternalForce="0,0,0") {
    angularDampingOverride=0.9
    bAngularSpring=true
    bGravityOverrideInSimSpace=true
    bLinearSpring=true
    bOverrideAngularDamping=true
    bOverrideLinearDamping=true
    gravityScale=0
    linearDampingOverride=0.9
    linearSpringConstant=10
    simulationSpace=BoneRelative
    alphaBoolBlend {blendOption=Linear}
    boundBone {boneName=stone_01_small}
    componentLinearAccScale=<0.2 0.2 0.2>
    componentLinearVelScale=<0.6 0.6 0.6>
    gravityOverride {z=-980}
    physicsBodyDefinitions [{
      sphereCollisionRadius=0
      boundBone {boneName=stone_01_small}
      boxExtents=<50 50 50>
      constraintSetup {angularLimitsMax=<0 0 0>, angularLimitsMin=<0 0 0>, angularTarget=<0 0 0>, linearAxesMax=<0.75 0.75 0.75>, linearAxesMin=<-0.75 -0.75 -0.75>}
      localJointOffset=<0 -3 0>
     }]
    relativeSpaceBone {boneName=handle}
   }
   n25 = Anim Dynamics - Bone: stone_02_small(ComponentPose=@n24, Alpha=0.8, ExternalForce="0,0,0") {
    angularDampingOverride=0.9
    bAngularSpring=true
    bGravityOverrideInSimSpace=true
    bLinearSpring=true
    bOverrideAngularDamping=true
    bOverrideLinearDamping=true
    gravityScale=0
    linearDampingOverride=0.9
    linearSpringConstant=10
    simulationSpace=BoneRelative
    alphaBoolBlend {blendOption=Linear}
    boundBone {boneName=stone_02_small}
    componentLinearAccScale=<0.2 0.2 0.2>
    componentLinearVelScale=<0.6 0.6 0.6>
    gravityOverride {z=-980}
    physicsBodyDefinitions [{
      sphereCollisionRadius=0
      boundBone {boneName=stone_02_small}
      boxExtents=<50 50 50>
      constraintSetup {angularLimitsMax=<0 0 0>, angularLimitsMin=<0 0 0>, angularTarget=<0 0 0>, linearAxesMax=<0.75 0.75 0.75>, linearAxesMin=<-0.75 -0.75 -0.75>}
      localJointOffset=<0 -3 0>
     }]
    relativeSpaceBone {boneName=handle}
   }
   n26 = Anim Dynamics - Bone: stone_03_small(ComponentPose=@n25, Alpha=0.8, ExternalForce="0,0,0") {
    angularDampingOverride=0.9
    bAngularSpring=true
    bGravityOverrideInSimSpace=true
    bLinearSpring=true
    bOverrideAngularDamping=true
    bOverrideLinearDamping=true
    gravityScale=0
    linearDampingOverride=0.9
    linearSpringConstant=10
    simulationSpace=BoneRelative
    alphaBoolBlend {blendOption=Linear}
    boundBone {boneName=stone_03_small}
    componentLinearAccScale=<0.2 0.2 0.2>
    componentLinearVelScale=<0.6 0.6 0.6>
    gravityOverride {z=-980}
    physicsBodyDefinitions [{
      sphereCollisionRadius=0
      boundBone {boneName=stone_03_small}
      boxExtents=<50 50 50>
      constraintSetup {angularLimitsMax=<0 0 0>, angularLimitsMin=<0 0 0>, angularTarget=<0 0 0>, linearAxesMax=<0.75 0.75 0.75>, linearAxesMin=<-0.75 -0.75 -0.75>}
      localJointOffset=<0 -3 0>
     }]
    relativeSpaceBone {boneName=handle}
   }
   n27 = Anim Dynamics - Bone: stone_04_small(ComponentPose=@n26, Alpha=0.8, ExternalForce="0,0,0") {
    angularDampingOverride=0.9
    bAngularSpring=true
    bGravityOverrideInSimSpace=true
    bLinearSpring=true
    bOverrideAngularDamping=true
    bOverrideLinearDamping=true
    gravityScale=0
    linearDampingOverride=0.9
    linearSpringConstant=10
    simulationSpace=BoneRelative
    alphaBoolBlend {blendOption=Linear}
    boundBone {boneName=stone_04_small}
    componentLinearAccScale=<0.2 0.2 0.2>
    componentLinearVelScale=<0.6 0.6 0.6>
    gravityOverride {z=-980}
    physicsBodyDefinitions [{
      sphereCollisionRadius=0
      boundBone {boneName=stone_04_small}
      boxExtents=<50 50 50>
      constraintSetup {angularLimitsMax=<0 0 0>, angularLimitsMin=<0 0 0>, angularTarget=<0 0 0>, linearAxesMax=<0.75 0.75 0.75>, linearAxesMin=<-0.75 -0.75 -0.75>}
      localJointOffset=<0 -3 0>
     }]
    relativeSpaceBone {boneName=handle}
   }
   n28 = Anim Dynamics - Bone: stone_05_small(ComponentPose=@n27, Alpha=0.8, ExternalForce="0,0,0") {
    angularDampingOverride=0.9
    bAngularSpring=true
    bGravityOverrideInSimSpace=true
    bLinearSpring=true
    bOverrideAngularDamping=true
    bOverrideLinearDamping=true
    gravityScale=0
    linearDampingOverride=0.9
    linearSpringConstant=10
    simulationSpace=BoneRelative
    alphaBoolBlend {blendOption=Linear}
    boundBone {boneName=stone_05_small}
    componentLinearAccScale=<0.2 0.2 0.2>
    componentLinearVelScale=<0.6 0.6 0.6>
    gravityOverride {z=-980}
    physicsBodyDefinitions [{
      sphereCollisionRadius=0
      boundBone {boneName=stone_05_small}
      boxExtents=<50 50 50>
      constraintSetup {angularLimitsMax=<0 0 0>, angularLimitsMin=<0 0 0>, angularTarget=<0 0 0>, linearAxesMax=<0.75 0.75 0.75>, linearAxesMin=<-0.75 -0.75 -0.75>}
      localJointOffset=<0 -3 0>
     }]
    relativeSpaceBone {boneName=handle}
   }
   n29 = Anim Dynamics - Bone: stone_06_small(ComponentPose=@n28, Alpha=0.8, ExternalForce="0,0,0") {
    angularDampingOverride=0.9
    bAngularSpring=true
    bGravityOverrideInSimSpace=true
    bLinearSpring=true
    bOverrideAngularDamping=true
    bOverrideLinearDamping=true
    gravityScale=0
    linearDampingOverride=0.9
    linearSpringConstant=10
    simulationSpace=BoneRelative
    alphaBoolBlend {blendOption=Linear}
    boundBone {boneName=stone_06_small}
    componentLinearAccScale=<0.2 0.2 0.2>
    componentLinearVelScale=<0.6 0.6 0.6>
    gravityOverride {z=-980}
    physicsBodyDefinitions [{
      sphereCollisionRadius=0
      boundBone {boneName=stone_06_small}
      boxExtents=<50 50 50>
      constraintSetup {angularLimitsMax=<0 0 0>, angularLimitsMin=<0 0 0>, angularTarget=<0 0 0>, linearAxesMax=<0.75 0.75 0.75>, linearAxesMin=<-0.75 -0.75 -0.75>}
      localJointOffset=<0 -3 0>
     }]
    relativeSpaceBone {boneName=handle}
   }
   n30 = Anim Dynamics - Bone: stone_07_small(ComponentPose=@n29, Alpha=0.8, ExternalForce="0,0,0") {
    angularDampingOverride=0.9
    bAngularSpring=true
    bGravityOverrideInSimSpace=true
    bLinearSpring=true
    bOverrideAngularDamping=true
    bOverrideLinearDamping=true
    gravityScale=0
    linearDampingOverride=0.9
    linearSpringConstant=10
    simulationSpace=BoneRelative
    alphaBoolBlend {blendOption=Linear}
    boundBone {boneName=stone_07_small}
    componentLinearAccScale=<0.2 0.2 0.2>
    componentLinearVelScale=<0.6 0.6 0.6>
    gravityOverride {z=-980}
    physicsBodyDefinitions [{
      sphereCollisionRadius=0
      boundBone {boneName=stone_07_small}
      boxExtents=<50 50 50>
      constraintSetup {angularLimitsMax=<0 0 0>, angularLimitsMin=<0 0 0>, angularTarget=<0 0 0>, linearAxesMax=<0.75 0.75 0.75>, linearAxesMin=<-0.75 -0.75 -0.75>}
      localJointOffset=<0 -3 0>
     }]
    relativeSpaceBone {boneName=handle}
   }
   n31 = Anim Dynamics - Bone: stone_08_small(ComponentPose=@n30, Alpha=0.8, ExternalForce="0,0,0") {
    angularDampingOverride=0.9
    bAngularSpring=true
    bGravityOverrideInSimSpace=true
    bLinearSpring=true
    bOverrideAngularDamping=true
    bOverrideLinearDamping=true
    gravityScale=0
    linearDampingOverride=0.9
    linearSpringConstant=10
    simulationSpace=BoneRelative
    alphaBoolBlend {blendOption=Linear}
    boundBone {boneName=stone_08_small}
    componentLinearAccScale=<0.2 0.2 0.2>
    componentLinearVelScale=<0.6 0.6 0.6>
    gravityOverride {z=-980}
    physicsBodyDefinitions [{
      sphereCollisionRadius=0
      boundBone {boneName=stone_08_small}
      boxExtents=<50 50 50>
      constraintSetup {angularLimitsMax=<0 0 0>, angularLimitsMin=<0 0 0>, angularTarget=<0 0 0>, linearAxesMax=<0.75 0.75 0.75>, linearAxesMin=<-0.75 -0.75 -0.75>}
      localJointOffset=<0 -3 0>
     }]
    relativeSpaceBone {boneName=handle}
   }
   n32 = Anim Dynamics - Bone: stone_09_small(ComponentPose=@n31, Alpha=0.8, ExternalForce="0,0,0") {
    angularDampingOverride=0.9
    bAngularSpring=true
    bGravityOverrideInSimSpace=true
    bLinearSpring=true
    bOverrideAngularDamping=true
    bOverrideLinearDamping=true
    gravityScale=0
    linearDampingOverride=0.9
    linearSpringConstant=10
    simulationSpace=BoneRelative
    alphaBoolBlend {blendOption=Linear}
    boundBone {boneName=stone_09_small}
    componentLinearAccScale=<0.2 0.2 0.2>
    componentLinearVelScale=<0.6 0.6 0.6>
    gravityOverride {z=-980}
    physicsBodyDefinitions [{
      sphereCollisionRadius=0
      boundBone {boneName=stone_09_small}
      boxExtents=<50 50 50>
      constraintSetup {angularLimitsMax=<0 0 0>, angularLimitsMin=<0 0 0>, angularTarget=<0 0 0>, linearAxesMax=<0.75 0.75 0.75>, linearAxesMin=<-0.75 -0.75 -0.75>}
      localJointOffset=<0 -3 0>
     }]
    relativeSpaceBone {boneName=handle}
   }
   Output Pose(Result=Component To Local(ComponentPose=@n32))
  ]}
 EventGraph {nodes [
   Event BlueprintUpdateAnimation()
   > n1 = Sequence!(execute) -> (then_0, then_1, then_2:Exec)
   n2 = GetGameTimeInSeconds()
   n3 = "Sin (Radians)"(A=float * float(A=float * float(A=1, B=1.5), B=@n2))
   n4 = "Sin (Radians)"(A=float * float(A=float * float(A=1, B=1), B=@n2))
   n5 = "Sin (Radians)"(A=float * float(A=float * float(A=1, B=2), B=@n2))
   Set StoneFloatA(
    execute=@n1.then_1
    StoneFloatA:Vector=MakeVector(X=float * float(A=@n3, B=0.5), Y=float * float(A=@n4, B=0.6), Z=float * float(A=@n5, B=0.22))
   ) -> (Output_Get:Vector)
   n6 = "Cos (Radians)"!(A:double=float * float(A=float * float(A=1, B=1.5), B=@n2)) -> (ReturnValue:double)
   n7 = "Cos (Radians)"(A=float * float(A=float * float(A=1, B=1.78), B=@n2))
   n8 = "Cos (Radians)"(A=float * float(A=float * float(A=1, B=1.15), B=@n2))
   > Set StoneFloatB(StoneFloatB:Vector=MakeVector(X=float * float(A=@n6, B=0.6), Y=float * float(A=@n7, B=0.725), Z=float * float(A=@n8, B=0.85))) -> (Output_Get:Vector)
   n9 = "Sin (Radians)"(A=float * float(A=float * float(A=1, B=1.5), B=@n2))
   n10 = "Sin (Radians)"(A=float * float(A=float * float(A=1, B=1), B=@n2))
   n11 = "Sin (Radians)"(A=float * float(A=float * float(A=1, B=2), B=@n2))
   > Set StoneFloatC(StoneFloatC:Vector=MakeVector(X=float * float(A=@n9, B=1.5), Y=float * float(A=@n10, B=0.6), Z=float * float(A=@n11, B=0.25))) -> (Output_Get:Vector)
   n12 = float / float(A=@n2, B=-5) {is=title K2Node_CallFunction}
   n13 = float / float(A=@n2, B=-20)
   > Set Rotator1(Rotator1:Rotator=MakeRotator(Roll=float * float(A=@n12, B=100), Yaw=float * float(A=@n13, B=1000))) -> (Output_Get:Rotator)
   n14 = float / float(A=@n2, B=-500)
   > Set Rotator2(Rotator2:Rotator=MakeRotator(Yaw=float * float(A=@n14, B=20000))) -> (Output_Get:Rotator)
   n15 = float / float(A=@n2, B=-5)
   n16 = float / float(A=@n2, B=-10)
   > Set Rotator3(Rotator3:Rotator=MakeRotator(Pitch=float * float(A=@n15, B=-100), Yaw=float * float(A=@n16, B=-500))) -> (Output_Get:Rotator)
   n17 = "Sin (Radians)"(A=float * float(A=float * float(A=1, B=1.5), B=@n2))
   n18 = "Sin (Radians)"(A=float * float(A=float * float(A=1, B=0.95), B=@n2))
   n19 = "Sin (Radians)"(A=float * float(A=float * float(A=0.6, B=1.5), B=@n2))
   > Set Shard_Float_D(Shard_Float_D:Vector=MakeVector(X=float * float(A=@n19, B=0.85), Y=float * float(A=@n18, B=0.1), Z=float * float(A=@n17, B=0.7))) -> (Output_Get:Vector)
   TryGetPawnOwner()
  ]}
]
variables [
 Base_Rotator {type=Rotator}
 BaseFloater {type=Vector}
 Rotator1 {type=Rotator}
 Rotator2 {type=Rotator}
 Rotator3 {type=Rotator}
 Shard_Float_D {type=Vector}
 StoneFloatA {type=Vector}
 StoneFloatB {type=Vector}
 StoneFloatC {type=Vector}
]
